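# Physics asset for the Citadel male medium armor's back skirt cape, by its name the skydive variant.
# Capsule bodies form three chains: left and right hip chains from the HipPad anchors and a longer tail chain from Tail.
# Chain links have narrow swing limits and zero twist, while tiny anchor capsules join root with free rotation.
# Thigh and right-calf capsules (no left calf) add leg volumes, alongside a very large offset root capsule.
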PhysicsAsset /BRCosmetics/Characters/Player/Male/Medium/Bodies/M_MED_Citadel_Armor/Meshes/M_MED_Citadel_Armor_BackSkirtCapeSkydive_Physics1
bodies {
 Hip1_dyn_l [{shapeType=Capsule, length=6.25, radius=9.25, shapeName=Hip1_dyn_l_capsule, center=<5 0 8>, rotation=<rot -90 0 0>}]
 Hip1_dyn_r [{shapeType=Capsule, length=6.25, radius=9.25, shapeName=Hip1_dyn_r_capsule, center=<-5 0 -8>, rotation=<rot -90 0 0>}]
 Hip2__dyn_l [{shapeType=Capsule, length=6.25, radius=9.25, shapeName=Hip2__dyn_l_capsule, center=<5 0 8>, rotation=<rot -90 0 0>}]
 Hip2_dyn_r [{shapeType=Capsule, length=6.25, radius=9.25, shapeName=Hip2_dyn_r_capsule, center=<-5 0 -8>, rotation=<rot -90 0 0>}]
 Hip3__dyn_l [{shapeType=Capsule, length=19.75, radius=9.25, shapeName=Hip3__dyn_l_capsule, center=<5 0 8>, rotation=<rot -90 0 0>}]
 Hip3_dyn_r [{shapeType=Capsule, length=6.25, radius=9.25, shapeName=Hip3_dyn_r_capsule, center=<-5 0 -8>, rotation=<rot -90 0 0>}]
 HipPad_l [{shapeType=Capsule, length=0.1, radius=0.1, shapeName=HipPad_l_capsule, rotation=<rot -90 90 -89.9998>}]
 HipPad_r [{shapeType=Capsule, length=0.1, radius=0.1, shapeName=HipPad_r_capsule, rotation=<rot -90 90 -89.9998>}]
 Tail [{shapeType=Capsule, length=0.1, radius=0.1, shapeName=Tail_capsule, rotation=<rot 0 -90 0>}]
 Tail1_dyn [{shapeType=Capsule, length=6.25, radius=9.25, shapeName=Tail1_dyn_capsule, center=<5 0 8>, rotation=<rot -90 0 0>}]
 Tail2_dyn [{shapeType=Capsule, length=6.25, radius=9.25, shapeName=Tail2_dyn_capsule, center=<5 0 8>, rotation=<rot -90 0 0>}]
 Tail3_dyn [{shapeType=Capsule, length=6.25, radius=9.25, shapeName=Tail3_dyn_capsule, center=<5 0 8>, rotation=<rot -90 0 0>}]
 Tail4_dyn [{shapeType=Capsule, length=6.25, radius=9.25, shapeName=Tail4_dyn_capsule, center=<5 0 8>, rotation=<rot -90 0 0>}]
 calf_r [{shapeType=Capsule, length=15.8182, radius=9.283, shapeName=calf_r_capsule, center=<27.9611 2.8516 0.6805>, rotation=<rot -90 -0.0792 0.079>}]
 root [{shapeType=Capsule, length=0, radius=200, shapeName=root_capsule, center=<0 -88.3298 -180.9445>, rotation=<rot 0 -90 0>}]
 thigh_l [{shapeType=Capsule, length=8.4611, radius=10.3633, shapeName=thigh_l_capsule, center=<-40 0 0>, rotation=<rot 90 0 0>}]
 thigh_r [{shapeType=Capsule, length=8.4611, radius=10.3633, shapeName=thigh_r_capsule, center=<40 0 0>, rotation=<rot 90 0 0>}]
}
constraints [
 {bone1Name=Hip1_dyn_l, bone2Name=HipPad_l, swing1LimitDegrees=20, swing2LimitDegrees=20, twistLimitDegrees=0}
 {bone1Name=Hip2__dyn_l, bone2Name=Hip1_dyn_l, swing1LimitDegrees=20, swing2LimitDegrees=20, twistLimitDegrees=0}
 {bone1Name=Hip3__dyn_l, bone2Name=Hip2__dyn_l, swing1LimitDegrees=20, swing2LimitDegrees=20, twistLimitDegrees=0}
 {bone1Name=Hip1_dyn_r, bone2Name=HipPad_r, swing1LimitDegrees=20, swing2LimitDegrees=20, twistLimitDegrees=0}
 {bone1Name=Hip2_dyn_r, bone2Name=Hip1_dyn_r, swing1LimitDegrees=20, swing2LimitDegrees=20, twistLimitDegrees=0}
 {bone1Name=Hip3_dyn_r, bone2Name=Hip2_dyn_r, swing1LimitDegrees=20, swing2LimitDegrees=20, twistLimitDegrees=0}
 {bone1Name=Tail1_dyn, bone2Name=Tail, swing1LimitDegrees=20, swing2LimitDegrees=20, twistLimitDegrees=0}
 {bone1Name=Tail2_dyn, bone2Name=Tail1_dyn, swing1LimitDegrees=20, swing2LimitDegrees=20, twistLimitDegrees=0}
 {bone1Name=Tail3_dyn, bone2Name=Tail2_dyn, swing1LimitDegrees=20, swing2LimitDegrees=20, twistLimitDegrees=0}
 {bone1Name=Tail4_dyn, bone2Name=Tail3_dyn, swing1LimitDegrees=20, swing2LimitDegrees=20, twistLimitDegrees=0}
 {
  bone1Name=calf_r
  bone2Name=thigh_r
  swing1LimitDegrees=45
  swing1Motion=Free
  swing2LimitDegrees=45
  swing2Motion=Free
  twistLimitDegrees=45
  twistMotion=Free
 }
 {
  bone1Name=thigh_l
  bone2Name=root
  swing1LimitDegrees=45
  swing1Motion=Free
  swing2LimitDegrees=45
  swing2Motion=Free
  twistLimitDegrees=45
  twistMotion=Free
 }
 {
  bone1Name=thigh_r
  bone2Name=root
  swing1LimitDegrees=45
  swing1Motion=Free
  swing2LimitDegrees=45
  swing2Motion=Free
  twistLimitDegrees=45
  twistMotion=Free
 }
 {
  bone1Name=HipPad_l
  bone2Name=root
  swing1LimitDegrees=45
  swing1Motion=Free
  swing2LimitDegrees=45
  swing2Motion=Free
  twistLimitDegrees=45
  twistMotion=Free
 }
 {
  bone1Name=HipPad_r
  bone2Name=root
  swing1LimitDegrees=45
  swing1Motion=Free
  swing2LimitDegrees=45
  swing2Motion=Free
  twistLimitDegrees=45
  twistMotion=Free
 }
 {
  bone1Name=Tail
  bone2Name=root
  swing1LimitDegrees=45
  swing1Motion=Free
  swing2LimitDegrees=45
  swing2Motion=Free
  twistLimitDegrees=45
  twistMotion=Free
 }
]
