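# Animation Blueprint for the face accessory part of the male medium Citadel Armor, adjusting head pose on top of the pose copied from its attached parent mesh.
# Each update derives alphas from head-to-spine_05 transform deltas, plus a flag for crouching, stationary and targeting.
# These alphas drive three interpolated, additive bone-space head rotations in series.
# The accessory's Control Rig processes the result before output.
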AnimBlueprint /BRCosmetics/Characters/Player/Male/Medium/Bodies/M_MED_Citadel_Armor/Meshes/Parts/M_MED_Citadel_Armor_FaceAcc_AnimBP
target_skeleton=&/M_MED_Citadel_Armor_FaceAcc_Skeleton
graphs [
 AnimGraph {nodes [
   n1 = "Transform (Modify) Bone - Bone: head"(
    ComponentPose=Local To Component(LocalPose=Copy Pose From Mesh() {bUseAttachedParent=true})
    Alpha=Get headRotBack() -> (headRotBack:double)
    Translation="0,0,0"
    Rotation="0,20,0"
    Scale="1,1,1"
   ) {
    rotationMode=BMM_Additive
    rotationSpace=BCS_BoneSpace
    scaleMode=BMM_Additive
    scaleSpace=BCS_BoneSpace
    translationMode=BMM_Additive
    translationSpace=BCS_BoneSpace
    alphaBoolBlend {blendOption=Linear}
    alphaScaleBiasClamp {bInterpResult=true, interpSpeedDecreasing=20, interpSpeedIncreasing=20}
    boneToModify {boneName=head}
    rotation {yaw=20}
   }
   n2 = "Transform (Modify) Bone - Bone: head"(ComponentPose=@n1, Alpha=Get headBackxTwist() -> (headBackxTwist:double), Translation="0,0,0", Rotation="-10,0,0", Scale="1,1,1") {
    rotationMode=BMM_Additive
    rotationSpace=BCS_BoneSpace
    scaleMode=BMM_Additive
    scaleSpace=BCS_BoneSpace
    translationMode=BMM_Additive
    translationSpace=BCS_BoneSpace
    alphaBoolBlend {blendOption=Linear}
    alphaScaleBiasClamp {bInterpResult=true, interpSpeedDecreasing=20, interpSpeedIncreasing=20}
    boneToModify {boneName=head}
    rotation {pitch=-10}
   }
   n3 = "Transform (Modify) Bone - Bone: head"(
    ComponentPose=@n2
    Alpha=Get Is Crouching NotMoving Float() -> (Is Crouching NotMoving Float:double)
    Translation="0,0,0"
    Rotation="-10,0,0"
    Scale="1,1,1"
   ) {
    rotationMode=BMM_Additive
    rotationSpace=BCS_BoneSpace
    scaleMode=BMM_Additive
    scaleSpace=BCS_BoneSpace
    translationMode=BMM_Additive
    translationSpace=BCS_BoneSpace
    alphaBoolBlend {blendOption=Linear}
    alphaScaleBiasClamp {bInterpResult=true, interpSpeedDecreasing=20, interpSpeedIncreasing=20}
    boneToModify {boneName=head}
    rotation {pitch=-10}
   }
   n4 = Control Rig(Source=Component To Local(ComponentPose=@n3)) {
    alphaBoolBlend {blendOption=Linear}
    controlRigAssetReference {blueprintRigClass=&/M_MED_Citadel_Armor_FaceAcc_CtrlRig.M_MED_Citadel_Armor_FaceAcc_CtrlRig_C}
   }
   Output Pose(Result=@n4) {layerGroup=None}
  ]}
 EventGraph {nodes [
   Event BlueprintUpdateAnimation()
   n1 = GetAlphaFromReferenceTransformDelta(SkeletalMeshComponent=GetOwnerSkeletalMeshComponent(), BoneName=head, BaseName=spine_05, InputRange=(X=20,Y=-50), TwistAxis="1,0,0")
   > Set headRotBack(headRotBack:double=@n1) -> (Output_Get:double)
   > Set headBackxTwist(headBackxTwist:double=float * float(
     A=@n1
     B=GetAlphaFromReferenceTransformDelta(BoneName=head, BaseName=spine_05, DeltaComponentType=EulerY, InputRange=(X=0,Y=80), TwistAxis="1,0,0")
    )) -> (Output_Get:double)
   n2 = NOT Boolean(A=Get bIsMoving2D())
   n3 = Get bIsCrouching()
   n4 = Get bIsTargeting()
   > Set Is Crouching NotMoving Float(Is Crouching NotMoving Float:double="To Float (Boolean)"(InBool=AND Boolean!(A=@n3, B=@n2, C:bool=@n4) -> (ReturnValue))) -> (Output_Get:double)
  ]}
]
variables [
 headBackxTwist {type=number}
 headRotBack {type=number}
 Is Crouching NotMoving Float {type=number}
]
